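# Level Sequence in the NeverCook TechAnim folder that frames a static shot, apparently for cape alignment renders.
# A CineCameraActor with a spawn track holds a fixed transform, with no keyframes or camera component animation.
# A Camera Cut track starts 45 frames before zero, likely as warm-up, and ends at frame 1.
# A Level Visibility track has two sections covering that single frame at 30 fps.
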LevelSequence /Game/NeverCook/TechAnim/CapeAlignment_Render/CapeAlignment_Render_Sequence
bindings [
 CineCameraActor {
  id=#1
  possessed_class=CineCameraActor
  tracks [
   {
    type=MovieScene3DTransformTrack
    sections [{
      type=MovieScene3DTransformSection
      channels [
       Location.X {type=MovieSceneScriptingDoubleChannel, default=-499.7798}
       Location.Y {type=MovieSceneScriptingDoubleChannel, default=1123.6081}
       Location.Z {type=MovieSceneScriptingDoubleChannel, default=-41.0177}
       Rotation.X {type=MovieSceneScriptingDoubleChannel, default=0}
       Rotation.Y {type=MovieSceneScriptingDoubleChannel, default=-0.9997}
       Rotation.Z {type=MovieSceneScriptingDoubleChannel, default=-52.5991}
       Scale.X {type=MovieSceneScriptingDoubleChannel, default=1}
       Scale.Y {type=MovieSceneScriptingDoubleChannel, default=1}
       Scale.Z {type=MovieSceneScriptingDoubleChannel, default=1}
      ]
     }]
   }
   {type=MovieSceneSpawnTrack, sections [{type=MovieSceneSpawnSection, channels [None {type=MovieSceneScriptingBoolChannel, default=true}]}]}
  ]
 }
 CameraComponent {id=#2, parent_id=#1, possessed_class=CineCameraComponent, tracks []}
]
display_rate {denominator=1, numerator=30}
tick_resolution {denominator=1, numerator=24000}
tracks [
 {
  type=MovieSceneLevelVisibilityTrack
  sections [{type=MovieSceneLevelVisibilitySection, end_frame=1, start_frame=0}, {type=MovieSceneLevelVisibilitySection, end_frame=1, start_frame=0}]
 }
 {type=MovieSceneCameraCutTrack, sections [{type=MovieSceneCameraCutSection, end_frame=1, start_frame=-45}]}
]
